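# Physics asset for the SK_ATT_Wedding skeletal mesh, an attachment for the Wedding vehicle cosmetic body.
# It holds a single body made of one capsule, offset and rotated from the bone origin.
# It has no constraints, so it provides one simple collision and bounds volume with no jointed ragdoll setup.
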PhysicsAsset /VehicleCosmetics/Bodies/Wedding/Meshes/Attachments/SK_ATT_Wedding_PhysicsAsset
bodies {body [{shapeType=Capsule, length=88.8038, radius=51.242, center=<-32.4864 0.3946 82.1264>, rotation=<rot 0 -177.292 -88.966>}]}
constraints []
